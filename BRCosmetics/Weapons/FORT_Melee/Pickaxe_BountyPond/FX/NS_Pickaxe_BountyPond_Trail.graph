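# Niagara swing trail for the BountyPond pickaxe, layering pink-purple anim-trail ribbons with pixel sparks.
# Five ribbon emitters (edges, pixel, shadow, side corners, wispy) stretch between two system socket indices, with alpha driven by system fade-in and fade-out values.
# A sprite emitter spawns short-lived sparks from a socket reader by rate and distance traveled, inheriting slight velocity.
# Time-of-day color tints the layers, while random per-particle and system-wide RGB flicker add color glitching.
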NiagaraSystem /BRCosmetics/Weapons/FORT_Melee/Pickaxe_BountyPond/FX/NS_Pickaxe_BountyPond_Trail
dependencies {
 usedDataInterfaces [
  &NiagaraDataInterfaceColorCurve
  &NiagaraDataInterfaceCurve
  &NiagaraDataInterfaceVector2DCurve
  &NiagaraDataInterfaceSocketReader
 ]
 usedDynamicInputs [
  &Vector2DFromFloat
  &FN_FrontEndFloatModifier
  &FNTimeOfDayColor
  &MergeHelper_FloatSwitch
  &MergeHelper_IntSwitch
 ]
 usedModules [
  &self:SystemSpawnScript.SystemScriptSource.SystemScriptGraph.NiagaraNodeAssignment_1.SetVariables_#1
  &SystemState
  &self:SystemSpawnScript.SystemScriptSource.SystemScriptGraph.NiagaraNodeAssignment_4.SetVariables_#2
  &self:SystemSpawnScript.SystemScriptSource.SystemScriptGraph.NiagaraNodeAssignment_7.SetVariables_#3
  &self:SystemSpawnScript.SystemScriptSource.SystemScriptGraph.NiagaraNodeAssignment_0.SetVariables_#4
 ]
}
emitters {
 Particles_Pixel {
  emitterUpdateScript [
   &EmitterState(Life Cycle Mode=::NewEnumerator1, Loop Behavior=::NewEnumerator1, Loop Delay=0.01, Loop Duration=1, UseLoopDelay=true)
   &SpawnRate(Spawn Probability=0.9, SpawnRate=&Multiply_Float(A=16, B=$System.FadeIn), Use Spawn Probability=false)
   &SpawnPerUnit(Module.Use Max Movement Threshold=true, Module.Use Movement Tolerance=false, Spawn Spacing=40, Use Spawn Probability=false)
  ]
  particleSpawnScript [
   &InitializeParticle(
    Color Maximum=<rgba 1.1951 0.2406 1.2 1>
    Color Minimum=<rgba 1.75 0.1075 0.3982 1>
    Color Mode=::NewEnumerator2
    Lifetime Max=0.4
    Lifetime Min=0.15
    Lifetime Mode=::NewEnumerator1
    Sprite Size Max=<4.5 4.9>
    Sprite Size Min=<2.7 3.6>
    Sprite Size Mode=::NewEnumerator2
    UsePositionOffset=false
    Write Material Random=true
   )
   &SocketLocation(
    Interpolate Socket Sampling=true
    OverrideRandomSeed=false
    SamplePosOffset=false
    Socket Reader=$System.Socket Reader
    Socket Sampling Mode=::NewEnumerator1
    VelocitySpeedLimit=false
   )
   &InheritVelocity(Inherited Velocity Amount Scale=<0.02 0.02 0.02>, Speed Limit=true, UseSpeedThreshold=false)
   &AddVelocity(Rotation Coordinate Space=::World, Velocity=$Output.SocketLocation.SampledVelocity, Velocity Speed Scale=0.02)
  ]
  particleUpdateScript [
   &ParticleState()
   /SetVariables(Particles.RGB Flicker:Vector3f=&MakeVectorFromLinearColorRGB(Color=&ColorFromCurve(
      ColorCurve=DataInterface() {propertyValues {
        AlphaCurve {keys [<at 0 0.9 RCIM_Constant>]}
        BlueCurve {keys [<at 0 0.9 RCIM_Constant>,<at 0.1839 0.9 RCIM_Constant>,<at 0.2906 0.9 RCIM_Constant>,<at 0.4069 2 RCIM_Constant>,<at 0.5295 2 RCIM_Constant>,<at 0.7732 2 RCIM_Constant>]}
        GreenCurve {keys [<at 0 0.9 RCIM_Constant>,<at 0.1839 2 RCIM_Constant>,<at 0.2906 2 RCIM_Constant>,<at 0.4069 2 RCIM_Constant>,<at 0.5295 0.9 RCIM_Constant>,<at 0.7732 0.9 RCIM_Constant>]}
        RedCurve {keys [<at 0 2 RCIM_Constant>,<at 0.1839 2 RCIM_Constant>,<at 0.2906 0.9 RCIM_Constant>,<at 0.4069 0.9 RCIM_Constant>,<at 0.5295 0.9 RCIM_Constant>,<at 0.7732 2 RCIM_Constant>]}
       }}
      CurveIndex=&RandomRangeFloat(Minimum=0, Maximum=1, Evaluation Type=::NewEnumerator1)
     ))) {isSetParameters=true}
   &ScaleColor(
    Scale Mode=::NewEnumerator1
    Scale RGBA=&Multiply_Vector4(
     A=&MakeVector4_FromLinearColor(LinearColor=$System.TODColor)
     B=&MakeVector4FromVectorAndFloat("Vector (RGB)"=$Particles.RGB Flicker, "Float (Alpha)"=1)
    )
    ScaleRGBA=true
   )
   /SetVariables(
    Particles.GlowColor=&MultiplyLinearColor(
     Linear Color=$Particles.Initial.Color
     Scale Factor=&MakeVector4FromVectorAndFloat(
      "Vector (RGB)"=$Particles.RGB Flicker
      "Float (Alpha)"=&FloatFromCurve(
       FloatCurve=DataInterface() {propertyValues {Curve {
          defaultValue=1
          keys [
           {tangentMode=RCTM_User, time=0, value=1}
           {arriveTangent=-0.0405, interpMode=RCIM_Cubic, leaveTangent=-0.0405, tangentMode=RCTM_User, time=0.7009, value=1}
           {arriveTangent=-8.0089, interpMode=RCIM_Cubic, leaveTangent=-8.0089, tangentMode=RCTM_User, time=1, value=0}
          ]
         }}}
       Scale Curve=&FNTimeOfDayScalar(Night Scale=0.05, Morning Scale=0.05, Day Scale=0.05, Evening Scale=0.05)
      )
     )
    )
    Particles.GlowSize=&Multiply_Vector2D(
     A=$Particles.Initial.SpriteSize
     B=&Vector2DFromCurve(
      Vector2Curve=DataInterface() {propertyValues {
        XCurve {keys [
          {arriveTangent=-0.0322, interpMode=RCIM_Cubic, leaveTangent=-0.0322, tangentMode=RCTM_User, time=0, value=1.2}
          {arriveTangent=-1.9075, interpMode=RCIM_Cubic, leaveTangent=-1.9075, tangentMode=RCTM_User, time=0.5597, value=0.7556}
          <at 1 0.5 RCIM_Cubic>
         ]}
        YCurve {keys [{arriveTangent=0.6579, interpMode=RCIM_Cubic, leaveTangent=0.6579, tangentMode=RCTM_User, time=0, value=0.4852}, <at 1 0.8 RCIM_Cubic>]}
       }}
      ScaleCurve=<8 6>
     )
    )
   ) {isSetParameters=true}
   &ScaleSpriteSize(
    Non-Uniform Curve Sprite Scale=DataInterface() {propertyValues {
      XCurve {keys [
        {arriveTangent=-0.0322, interpMode=RCIM_Cubic, leaveTangent=-0.0322, tangentMode=RCTM_User, time=0, value=1}
        {arriveTangent=-1.9075, interpMode=RCIM_Cubic, leaveTangent=-1.9075, tangentMode=RCTM_User, time=0.572, value=0.6082}
        <at 1 0 RCIM_Cubic>
       ]}
      YCurve {keys [{arriveTangent=0.6579, interpMode=RCIM_Cubic, leaveTangent=0.6579, tangentMode=RCTM_User, time=0, value=0.4852}, <at 1 1.1 RCIM_Cubic>]}
     }}
    Scale Sprite Size Mode=::NewEnumerator3
   )
   &Drag(Drag=&RandomRangeFloat(Minimum=2, Maximum=4), Use Linear Drag=true, Use Rotational Drag=false)
   &SolveForcesAndVelocity(Clamp Velocity=false, Limit Acceleration=false)
  ]
  rendererClasses [&NiagaraSpriteRendererProperties]
  renderers [
   {rendererClass=&NiagaraSpriteRendererProperties}
   {rendererClass=&NiagaraSpriteRendererProperties, rendererIndex=1}
  ]
 }
 Trail_Edges {
  emitterUpdateScript [
   &EmitterState()
   &FN_SpawnAnimTrail(AnimTrailWidthMode=::FromFirst, Number of Anim Trails=2)
  ]
  particleSpawnScript [
   &FN_InitializeAnimTrailBoneIndecies(First Bone or Socket Index:NiagaraInt32=$System.Socket1, Second Bone or Socket Index:NiagaraInt32=$System.Socket2)
   &FN_InitializeAnimTrailBoneIndecies(AnimTrail Selection=1, First Bone or Socket Index:NiagaraInt32=$System.Socket2, Second Bone or Socket Index:NiagaraInt32=$System.Socket1) {module=FN_InitializeAnimTrailBoneIndecies001}
   &FN_InitializeAnimTrail(
    AnimTrailWidth=0.2
    Color=&MakeLinearColorFromVectorAndFloat(
     "Vector (RGB)"=&MakeVectorFromLinearColorRGB(Color=$System.TODColor)
     "Float (Alpha)"=&Multiply_Float(A=&MaskFloatBySpawnGroup(Masked Float=0.5, Passthrough Float=1), B=$System.FadeIn)
    )
    Lifetime=0.13
   )
  ]
  particleUpdateScript [
   &FN_UpdateAnimTrail() {module=FN_UpdateAnimTrail001}
   &ScaleColor(
    Scale Mode=::NewEnumerator1
    Scale RGBA=&Multiply_Vector4(
     A=&MakeVector4_FromLinearColor(LinearColor=&ColorFromCurve(ColorCurve=DataInterface() {propertyValues {
         AlphaCurve {keys [<at 0 1 RCIM_Cubic>]}
         BlueCurve {keys [<at 0 0.9047 RCIM_Cubic>,<at 1 0.0411 RCIM_Cubic>]}
         GreenCurve {keys [<at 0 0.0823 RCIM_Cubic>,<at 1 0.0043 RCIM_Cubic>]}
         RedCurve {keys [<at 0 1 RCIM_Cubic>,<at 1 1 RCIM_Cubic>]}
        }}))
     B=&MakeVector4FromVectorAndFloat("Vector (RGB)"=$System.GlobalRGBFLicker, "Float (Alpha)"=$System.FadeOut)
    )
    ScaleRGBA=true
   )
  ]
  rendererClasses [&NiagaraRibbonRendererProperties]
  renderers [{rendererClass=&NiagaraRibbonRendererProperties}]
 }
 Trail_Pixel {
  emitterUpdateScript [
   &EmitterState()
   &FN_SpawnAnimTrail()
  ]
  particleSpawnScript [
   &FN_InitializeAnimTrailBoneIndecies(First Bone or Socket Index:NiagaraInt32=$System.Socket1, Second Bone or Socket Index:NiagaraInt32=$System.Socket2)
   &FN_InitializeAnimTrail(
    Color=&MultiplyLinearColor(
     Linear Color=$System.TODColor
     Scale Factor=&MakeVector4FromVectorAndFloat("Vector (RGB)"=<1 1 1>, "Float (Alpha)"=&Multiply_Float(A=1, B=$System.FadeIn))
    )
    Lifetime=0.1
   )
  ]
  particleUpdateScript [
   &FN_UpdateAnimTrail()
   &ScaleColor(
    Scale Alpha=&Multiply_Float(A=$System.FadeOut, B=2)
    Scale RGB=&MakeVectorFromLinearColorRGB(Color=<rgba 0.2047 0.0378 0.4783 1>)
    ScaleA=true
    ScaleRGB=true
   )
   /SetVariables(Particles.TipColor:LinearColor=&MultiplyLinearColor(
     Linear Color=$Particles.Initial.Color
     Scale Factor=&MakeVector4FromVectorAndFloat(
      "Vector (RGB)"=&MakeVectorFromLinearColorRGB(Color=<rgba 12 2.3613 9.9631 1>)
      "Float (Alpha)"=$System.FadeOut
     )
    )) {isSetParameters=true}
  ]
  rendererClasses [&NiagaraRibbonRendererProperties]
  renderers [
   {rendererClass=&NiagaraRibbonRendererProperties}
   {rendererClass=&NiagaraRibbonRendererProperties, rendererIndex=1}
  ]
 }
 Trail_Shadow {
  emitterUpdateScript [
   &EmitterState()
   &FN_SpawnAnimTrail()
  ]
  particleSpawnScript [
   &FN_InitializeAnimTrailBoneIndecies(First Bone or Socket Index:NiagaraInt32=$System.Socket1, Second Bone or Socket Index:NiagaraInt32=$System.Socket2)
   &FN_InitializeAnimTrail(
    Color=&MakeLinearColorFromVectorAndFloat(
     "Vector (RGB)"=&MakeVectorFromLinearColorRGB(Color=<rgba 0.0174 0.0163 0.0172 1>)
     "Float (Alpha)"=&Multiply_Float(A=2, B=$System.FadeIn)
    )
    Lifetime=0.12
   )
  ]
  particleUpdateScript [
   &FN_UpdateAnimTrail() {module=FN_UpdateAnimTrail001}
   &ScaleColor(
    Scale Mode=::NewEnumerator1
    Scale RGBA=&Multiply_Vector4(
     A=&MakeVector4_FromLinearColor(LinearColor=$System.TODColor)
     B=&MakeVector4FromVectorAndFloat("Vector (RGB)"=$System.GlobalRGBFLicker, "Float (Alpha)"=$System.FadeOut)
    )
    ScaleRGBA=true
   )
  ]
  rendererClasses [&NiagaraRibbonRendererProperties]
  renderers [{rendererClass=&NiagaraRibbonRendererProperties}]
 }
 Trail_SideCorners {
  emitterUpdateScript [
   &EmitterState()
   &FN_SpawnAnimTrail(AnimTrailWidthMode=::FromFirst, Number of Anim Trails=2)
  ]
  particleSpawnScript [
   &FN_InitializeAnimTrailBoneIndecies(First Bone or Socket Index:NiagaraInt32=$System.Socket2, Second Bone or Socket Index:NiagaraInt32=$System.Socket1)
   &FN_InitializeAnimTrailBoneIndecies(AnimTrail Selection=1, First Bone or Socket Index:NiagaraInt32=$System.Socket2, Second Bone or Socket Index:NiagaraInt32=$System.Socket1) {module=FN_InitializeAnimTrailBoneIndecies001}
   &FN_InitializeAnimTrail(
    AnimTrailWidth=0.1
    Color=&MakeLinearColorFromVectorAndFloat(
     "Vector (RGB)"=&MakeVectorFromLinearColorRGB(Color=$System.TODColor)
     "Float (Alpha)"=&Multiply_Float(A=0.2, B=$System.FadeIn)
    )
    Lifetime=0.09
    Position Offset=&Multiply_VectorByFloat(Vector=<5 0 0>, Float=&MaskFloatBySpawnGroup(Masked Float=-1, Passthrough Float=1))
   )
  ]
  particleUpdateScript [
   &FN_UpdateAnimTrail() {module=FN_UpdateAnimTrail001}
   &ScaleColor(
    Scale Mode=::NewEnumerator1
    Scale RGBA=&Multiply_Vector4(
     A=&MakeVector4_FromLinearColor(LinearColor=&ColorFromCurve(ColorCurve=DataInterface() {propertyValues {
         AlphaCurve {keys [<at 0 1 RCIM_Cubic>]}
         BlueCurve {keys [<at 0 0.9047 RCIM_Cubic>,<at 1 0.0411 RCIM_Cubic>]}
         GreenCurve {keys [<at 0 0.0823 RCIM_Cubic>,<at 1 0.0043 RCIM_Cubic>]}
         RedCurve {keys [<at 0 1 RCIM_Cubic>,<at 1 1 RCIM_Cubic>]}
        }}))
     B=&MakeVector4FromVectorAndFloat("Vector (RGB)"=$System.GlobalRGBFLicker, "Float (Alpha)"=$System.FadeOut)
    )
    ScaleRGBA=true
   )
  ]
  rendererClasses [&NiagaraRibbonRendererProperties]
  renderers [{rendererClass=&NiagaraRibbonRendererProperties}]
 }
 Trail_Wispy {
  emitterUpdateScript [
   &EmitterState()
   &FN_SpawnAnimTrail()
  ]
  particleSpawnScript [
   &FN_InitializeAnimTrailBoneIndecies(First Bone or Socket Index:NiagaraInt32=$System.Socket1, Second Bone or Socket Index:NiagaraInt32=$System.Socket2)
   &FN_InitializeAnimTrail(
    AnimTrailWidth=1.4
    Color=&MultiplyLinearColor(
     Linear Color=$System.TODColor
     Scale Factor=&MakeVector4FromVectorAndFloat("Vector (RGB)"=<1 1 1>, "Float (Alpha)"=&Multiply_Float(A=0.2333, B=$System.FadeIn))
    )
    Lifetime=0.13
   )
  ]
  particleUpdateScript [
   &FN_UpdateAnimTrail() {module=FN_UpdateAnimTrail001}
   &ScaleColor(
    Scale Mode=::NewEnumerator1
    Scale RGBA=&Multiply_Vector4(
     A=&MakeVector4_FromLinearColor(LinearColor=&ColorFromCurve(ColorCurve=DataInterface() {propertyValues {
         AlphaCurve {keys [<at 0.1952 1 RCIM_Cubic>,<at 0.7233 0.498>]}
         BlueCurve {keys [
           <at 0 1.927>
           {arriveTangent=-2.0181, interpMode=RCIM_Cubic, leaveTangent=-2.0181, time=0.214, value=1.2472}
           <at 0.6525 0.6101>
           <at 0.9559 0.1997>
          ]}
         GreenCurve {keys [
           <at 0 0.9416>
           {arriveTangent=-1.1492, interpMode=RCIM_Cubic, leaveTangent=-1.1492, time=0.214, value=0.5503}
           <at 0.6525 0.1917>
           <at 0.9559 0.1168>
          ]}
         RedCurve {keys [<at 0 2>, {arriveTangent=-0.9533, interpMode=RCIM_Cubic, leaveTangent=-0.9533, time=0.214, value=1.4}, <at 0.6525 1.378>, <at 0.9559 1.4>]}
        }}))
     B=&MakeVector4FromVectorAndFloat("Vector (RGB)"=<1 1 1>, "Float (Alpha)"=$System.FadeOut)
    )
    ScaleRGBA=true
   )
  ]
  rendererClasses [&NiagaraRibbonRendererProperties]
  renderers [{rendererClass=&NiagaraRibbonRendererProperties}]
 }
}
summary {
 systemName=NS_Pickaxe_BountyPond_Trail
 emitters [
  {emitterName=Trail_Edges, rendererClasses [&NiagaraRibbonRendererProperties]}
  {emitterName=Trail_Pixel, rendererClasses [&NiagaraRibbonRendererProperties]}
  {emitterName=Particles_Pixel, rendererClasses [&NiagaraSpriteRendererProperties]}
  {emitterName=Trail_Wispy, rendererClasses [&NiagaraRibbonRendererProperties]}
  {emitterName=Trail_Shadow, rendererClasses [&NiagaraRibbonRendererProperties]}
  {emitterName=Trail_SideCorners, rendererClasses [&NiagaraRibbonRendererProperties]}
 ]
 userVariables [
  User.bIsFrontEnd {type=NiagaraBool}
  User.bIsFrontEndPreview {type=NiagaraBool}
  User.RibbonIDOffset {type=NiagaraInt32}
  User.SwingSectionIndex {type=NiagaraInt32}
 ]
}
